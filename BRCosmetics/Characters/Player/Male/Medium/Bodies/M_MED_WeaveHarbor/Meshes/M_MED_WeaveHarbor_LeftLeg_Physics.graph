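# Physics asset for the left leg mesh of the WeaveHarbor medium male player cosmetic body, covering the thigh and its dynamic shorts bones.
# Capsules wrap the thigh and upper shorts bones, while small spheres sit on the back shorts bone and its master bone.
# Two constraints with tight swing and twist limits chain back shorts to upper shorts to thigh.
# The master bone's body has no constraint.
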PhysicsAsset /BRCosmetics/Characters/Player/Male/Medium/Bodies/M_MED_WeaveHarbor/Meshes/M_MED_WeaveHarbor_LeftLeg_Physics
bodies {
 dyn_shortBack_l [{shapeType=Sphere, radius=2.4479, shapeName=dyn_shortBack_l_sphere}]
 dyn_shortBack_master_l [{shapeType=Sphere, radius=0.6709, shapeName=dyn_shortBack_master_l_sphere}]
 dyn_shortUpper_l [{
   shapeType=Capsule
   length=7.6326
   radius=7.1897
   shapeName=dyn_shortUpper_l_capsule
   center=<-7.3981 0.0099 0.1816>
   rotation=<rot -69.6949 -91.9175 91.7982>
  }]
 thigh_l [{
   shapeType=Capsule
   length=22.9996
   radius=7.0004
   shapeName=thigh_l_capsule
   center=<-16.0559 -0.1115 0.2826>
   rotation=<rot 89.2395 0 0.0002>
  }]
}
constraints [
 {bone1Name=dyn_shortBack_l, bone2Name=dyn_shortUpper_l, swing1LimitDegrees=3, swing2LimitDegrees=3, twistLimitDegrees=5}
 {bone1Name=dyn_shortUpper_l, bone2Name=thigh_l, swing1LimitDegrees=6, swing2LimitDegrees=6, twistLimitDegrees=3}
]
