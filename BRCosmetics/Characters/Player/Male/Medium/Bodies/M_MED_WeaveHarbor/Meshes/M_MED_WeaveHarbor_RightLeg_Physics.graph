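# Physics asset for the right leg mesh of the WeaveHarbor medium male body cosmetic, covering thigh and shorts "dyn" bones.
# Capsules cover the thigh and upper shorts bone, while spheres cover the back shorts bone and its master.
# Two separate constraints link upper shorts to thigh and back shorts to master, with tight swing and twist limits.
# Shape names carry left-side suffixes despite the right-side bones, suggesting a mirrored setup.
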PhysicsAsset /BRCosmetics/Characters/Player/Male/Medium/Bodies/M_MED_WeaveHarbor/Meshes/M_MED_WeaveHarbor_RightLeg_Physics
bodies {
 dyn_shortBack_master_r [{shapeType=Sphere, radius=0.6709, shapeName=dyn_shortBack_master_l_sphere}]
 dyn_shortBack_r [{shapeType=Sphere, radius=2.7458, shapeName=dyn_shortBack_l_sphere}]
 dyn_shortUpper_r [{
   shapeType=Capsule
   length=7.6326
   radius=7.1897
   shapeName=dyn_shortUpper_l_capsule
   center=<7.3981 -0.0099 -0.1816>
   rotation=<rot -69.6949 -91.9175 -88.2018>
  }]
 thigh_r [{
   shapeType=Capsule
   length=22.9996
   radius=7.0004
   shapeName=thigh_l_capsule
   center=<16.0559 0.1115 -0.2826>
   rotation=<rot 89.2395 0 -179.9998>
  }]
}
constraints [
 {bone1Name=dyn_shortUpper_r, bone2Name=thigh_r, swing1LimitDegrees=7, swing2LimitDegrees=7, twistLimitDegrees=3}
 {bone1Name=dyn_shortBack_r, bone2Name=dyn_shortBack_master_r, swing1LimitDegrees=3, swing2LimitDegrees=3, twistLimitDegrees=5}
]
