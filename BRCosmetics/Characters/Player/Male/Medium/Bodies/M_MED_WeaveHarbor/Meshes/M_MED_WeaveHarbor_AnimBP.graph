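# Post-process-style Anim Blueprint for the WeaveHarbor male cosmetic body, adding secondary motion over the pose copied from its parent mesh.
# The output chain runs Control Rig, bottle AnimDynamics, per-leg RigidBody sims, then CircleFX bone constraints.
# The event graph drives leg-sim alphas from smoothed knee bend and flags LOD 2+ for the rig.
# Shorts and bottle-pocket dynamics and state-driven bone offsets sit unconnected to the output.
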AnimBlueprint /BRCosmetics/Characters/Player/Male/Medium/Bodies/M_MED_WeaveHarbor/Meshes/M_MED_WeaveHarbor_AnimBP
target_skeleton=&/M_MED_WeaveHarbor_Skeleton
graphs [
 AnimGraph {nodes [
   n1 = Get Gravity Override()
   Anim Dynamics - Bone: dyn_short_r!(ComponentPose, bAlphaBoolEnabled:bool=True, ExternalForce=@n1) -> (Pose) {
    alphaInputType=Bool
    angularDampingOverride=1
    bOverrideAngularDamping=true
    bUseGravityOverride=true
    simulationSpace=RootRelative
    alphaBoolBlend {blendInTime=0.1, blendOption=Linear, blendOutTime=0.2}
    boundBone {boneName=dyn_short_r}
    componentAppliedLinearAccClamp=<2500 2500 2500>
    componentLinearAccScale=<0.6 0.6 0.6>
    componentLinearVelScale=<0.7 0.7 0.7>
    gravityOverride {z=-980}
    physicsBodyDefinitions [{
      boundBone {boneName=dyn_short_r}
      boxExtents=<15 15 15>
      constraintSetup {angularLimitsMax=<5 5 10>, angularLimitsMin=<-5 -5 -10>, angularTarget=<0 0 0>}
      localJointOffset=<5 0 0>
     }]
   }
   Anim Dynamics - Bone: dyn_short_l!(ComponentPose, bAlphaBoolEnabled:bool=True, ExternalForce=@n1) -> (Pose) {
    alphaInputType=Bool
    angularDampingOverride=1
    bOverrideAngularDamping=true
    bUseGravityOverride=true
    simulationSpace=RootRelative
    alphaBoolBlend {blendInTime=0.1, blendOption=Linear, blendOutTime=0.2}
    boundBone {boneName=dyn_short_l}
    componentAppliedLinearAccClamp=<2500 2500 2500>
    componentLinearAccScale=<0.6 0.6 0.6>
    componentLinearVelScale=<0.7 0.7 0.7>
    gravityOverride {z=-980}
    physicsBodyDefinitions [{
      boundBone {boneName=dyn_short_l}
      boxExtents=<15 15 15>
      constraintSetup {angularLimitsMax=<5 5 10>, angularLimitsMin=<-5 -5 -10>, angularTarget=<0 0 0>}
      localJointOffset=<5 0 0>
     }]
   }
   n2 = Anim Dynamics - Bone: dyn_short_l(bAlphaBoolEnabled=True, ExternalForce="0,0,0") {
    alphaInputType=Bool
    angularDampingOverride=1
    bOverrideAngularDamping=true
    bUseGravityOverride=true
    simulationSpace=RootRelative
    alphaBoolBlend {blendInTime=0.2, blendOption=Linear, blendOutTime=0.4}
    boundBone {boneName=dyn_short_l}
    componentAppliedLinearAccClamp=<3000 3000 3000>
    componentLinearAccScale=<0.1 0.1 0.1>
    componentLinearVelScale=<0.2 0.2 0.2>
    gravityOverride {z=-980}
    physicsBodyDefinitions [{
      boundBone {boneName=dyn_short_l}
      boxExtents=<5 20 20>
      constraintSetup {angularLimitsMax=<0 5 10>, angularLimitsMin=<0 -5 -10>, angularTarget=<0 0 0>}
      localJointOffset=<10 0 0>
     }]
   }
   Anim Dynamics - Bone: dyn_short_r(ComponentPose=@n2, bAlphaBoolEnabled=True, ExternalForce="0,0,0") {
    alphaInputType=Bool
    angularDampingOverride=1
    bOverrideAngularDamping=true
    bUseGravityOverride=true
    simulationSpace=RootRelative
    alphaBoolBlend {blendInTime=0.2, blendOption=Linear, blendOutTime=0.4}
    boundBone {boneName=dyn_short_r}
    componentAppliedLinearAccClamp=<3000 3000 3000>
    componentLinearAccScale=<0.1 0.1 0.1>
    componentLinearVelScale=<0.2 0.2 0.2>
    gravityOverride {z=-980}
    physicsBodyDefinitions [{
      boundBone {boneName=dyn_short_r}
      boxExtents=<5 20 20>
      constraintSetup {angularLimitsMax=<0 5 10>, angularLimitsMin=<0 -5 -10>, angularTarget=<0 0 0>}
      localJointOffset=<10 0 0>
     }]
   }
   n3 = Anim Dynamics - Bone: dyn_shortUpper_l!(ComponentPose, Alpha=1, GravityOverride:Vector="0,0,0", ExternalForce="0,0,0") -> (Pose) {
    angularDampingOverride=0.7
    angularSpringConstant=75
    bAngularSpring=true
    bOverrideAngularDamping=true
    bUseGravityOverride=true
    linearSpringConstant=200
    simulationSpace=BoneRelative
    alphaBoolBlend {blendOption=Linear}
    boundBone {boneName=dyn_shortUpper_l}
    componentLinearAccScale=<4 4 4>
    componentLinearVelScale=<4 4 4>
    physicsBodyDefinitions [{
      boundBone {boneName=dyn_shortUpper_l}
      boxExtents=<15 15 15>
      constraintSetup {angularLimitsMax=<0 5 3>, angularLimitsMin=<0 -5 -3>, angularTarget=<1 0 0>}
      localJointOffset=<-0.2 0 0>
     }]
    relativeSpaceBone {boneName=root}
   }
   n4 = Anim Dynamics - Bone: dyn_shortBack_l!(ComponentPose=@n3, Alpha=1, GravityOverride:Vector="0,0,0", ExternalForce="0,0,0") -> (Pose) {
    angularDampingOverride=0.85
    angularSpringConstant=50
    bLinearSpring=true
    bOverrideLinearDamping=true
    bUseGravityOverride=true
    linearDampingOverride=0.85
    linearSpringConstant=50
    simulationSpace=BoneRelative
    alphaBoolBlend {blendOption=Linear}
    boundBone {boneName=dyn_shortBack_l}
    componentLinearAccScale=<2.5 2.5 2.5>
    componentLinearVelScale=<2.5 2.5 2.5>
    physicsBodyDefinitions [{
      boundBone {boneName=dyn_shortBack_l}
      boxExtents=<10 10 10>
      constraintSetup {angularLimitsMax=<5 0 0>, angularLimitsMin=<-5 0 0>, angularTarget=<1 0 0>, linearAxesMax=<0 0 1.5>, linearAxesMin=<0 0 -1.5>}
      localJointOffset=<-0.1 0 0>
     }]
    relativeSpaceBone {boneName=dyn_shortUpper_l}
   }
   n5 = Anim Dynamics - Bone: dyn_shortUpper_r!(ComponentPose=@n4, Alpha=1, GravityOverride:Vector="0,0,0", ExternalForce="0,0,0") -> (Pose) {
    angularDampingOverride=0.8
    angularSpringConstant=300
    bAngularSpring=true
    bOverrideAngularDamping=true
    bUseGravityOverride=true
    linearSpringConstant=200
    simulationSpace=BoneRelative
    alphaBoolBlend {blendOption=Linear}
    boundBone {boneName=dyn_shortUpper_r}
    componentLinearAccScale=<4 4 4>
    componentLinearVelScale=<4 4 4>
    physicsBodyDefinitions [{
      boundBone {boneName=dyn_shortUpper_r}
      boxExtents=<10 10 10>
      constraintSetup {angularLimitsMax=<0 5 3>, angularLimitsMin=<0 -5 -3>, angularTarget=<0.9 0 0>}
      localJointOffset=<0.3 0 0>
     }]
    relativeSpaceBone {boneName=root}
   }
   Anim Dynamics - Bone: dyn_shortBack_r!(ComponentPose=@n5, Alpha=1, GravityOverride:Vector="0,0,0", ExternalForce="0,0,0") -> (Pose) {
    angularDampingOverride=0.85
    angularSpringConstant=50
    bLinearSpring=true
    bOverrideLinearDamping=true
    bUseGravityOverride=true
    linearDampingOverride=0.8
    linearSpringConstant=50
    simulationSpace=BoneRelative
    alphaBoolBlend {blendOption=Linear}
    boundBone {boneName=dyn_shortBack_r}
    componentLinearAccScale=<3 3 3>
    componentLinearVelScale=<3 3 3>
    physicsBodyDefinitions [{
      boundBone {boneName=dyn_shortBack_r}
      boxExtents=<10 10 10>
      constraintSetup {angularLimitsMax=<5 0 0>, angularLimitsMin=<-5 0 0>, angularTarget=<0.9 0 0>, linearAxesMax=<0 0 1.5>, linearAxesMin=<0 0 -1.5>}
      localJointOffset=<0.15 0 0>
     }]
    relativeSpaceBone {boneName=dyn_shortUpper_r}
   }
   "Constraint - dyn_bottle_poket_MSTR (PPPP)"!(
    ComponentPose
    Alpha=1
    ConstraintWeights_0=0.1619
    ConstraintWeights_1:float=0.0331
    ConstraintWeights_2:float=0.7819
    ConstraintWeights_3:float=0.0231
   ) -> (Pose) {
    alphaBoolBlend {blendOption=Linear}
    boneToModify {boneName=dyn_bottle_poket_MSTR}
    constraintSetup [
     {targetBone {boneName=pelvis}}
     {targetBone {boneName=thigh_l}}
     {targetBone {boneName=thigh_twist_01_l}}
     {targetBone {boneName=dyn_short_l}}
    ]
    constraintWeights [1,1,1,1]
   }
   n6 = Anim Dynamics - Bone: dyn_bottle_poket!(ComponentPose, bAlphaBoolEnabled:bool=Get Not in Sky Diving() -> (Not in Sky Diving:bool), ExternalForce="0,0,0") -> (Pose) {
    alphaInputType=Bool
    angularDampingOverride=1
    bOverrideAngularDamping=true
    bUseGravityOverride=true
    simulationSpace=RootRelative
    alphaBoolBlend {blendInTime=0.2, blendOption=Linear, blendOutTime=0.4}
    boundBone {boneName=dyn_bottle_poket}
    componentAppliedLinearAccClamp=<3000 3000 3000>
    componentLinearAccScale=<0.1 0.1 0.1>
    componentLinearVelScale=<0.2 0.2 0.2>
    gravityOverride {z=-980}
    physicsBodyDefinitions [{
      boundBone {boneName=dyn_bottle_poket}
      boxExtents=<10 20 20>
      constraintSetup {angularLimitsMax=<0 5 5>, angularLimitsMin=<0 -5 -5>, angularTarget=<0 0 0>}
      localJointOffset=<2 0 0>
     }]
   }
   n7 = "Transform (Modify) Bone - Bone: dyn_bottle_01"!(ComponentPose=@n6, bAlphaBoolEnabled:bool=Get bIsJumping(), Translation="-1,0,0", Rotation="0,0,0", Scale="1,1,1") -> (Pose) {
    alphaInputType=Bool
    rotationMode=BMM_Additive
    rotationSpace=BCS_BoneSpace
    scaleSpace=BCS_BoneSpace
    translationMode=BMM_Additive
    translationSpace=BCS_BoneSpace
    alphaBoolBlend {blendInTime=0.2, blendOption=Linear, blendOutTime=0.2}
    boneToModify {boneName=dyn_bottle_01}
    translation {x=-1}
   }
   n8 = "Transform (Modify) Bone - Bone: dyn_bottle_01"(ComponentPose=@n7, bAlphaBoolEnabled=Get Bottle falling() -> (Bottle falling:bool), Translation="1,0,0", Rotation="0,0,0", Scale="1,1,1") {
    alphaInputType=Bool
    rotationMode=BMM_Additive
    rotationSpace=BCS_BoneSpace
    scaleSpace=BCS_BoneSpace
    translationMode=BMM_Additive
    translationSpace=BCS_BoneSpace
    alphaBoolBlend {blendInTime=0.4, blendOption=Linear, blendOutTime=0.5}
    boneToModify {boneName=dyn_bottle_01}
    translation {x=1}
   }
   "Transform (Modify) Bone - Bone: dyn_bottle_01"(ComponentPose=@n8, bAlphaBoolEnabled=Get bIsSkydiveDiveMode(), Translation="-2,0,0", Rotation="0,0,0", Scale="1,1,1") {
    alphaInputType=Bool
    rotationMode=BMM_Additive
    rotationSpace=BCS_BoneSpace
    scaleSpace=BCS_BoneSpace
    translationMode=BMM_Additive
    translationSpace=BCS_BoneSpace
    alphaBoolBlend {blendInTime=0.5, blendOption=Linear, blendOutTime=0.2}
    boneToModify {boneName=dyn_bottle_01}
    translation {x=-2}
   }
   "Transform (Modify) Bone - Bone: dyn_bottle_poket"!(ComponentPose, bAlphaBoolEnabled:bool=Get bIsCrouching(), Translation="0,0,0", Rotation="0,-11,0", Scale="1,1,1") -> (Pose) {
    alphaInputType=Bool
    rotationMode=BMM_Additive
    rotationSpace=BCS_BoneSpace
    scaleSpace=BCS_BoneSpace
    translationMode=BMM_Additive
    translationSpace=BCS_BoneSpace
    alphaBoolBlend {blendInTime=0.2, blendOption=Linear, blendOutTime=0.2}
    boneToModify {boneName=dyn_bottle_poket}
    rotation {yaw=-11}
   }
   n9 = Control Rig!(Source=Copy Pose From Mesh() {bUseAttachedParent=true}, Alpha, IsLOD2:bool=Get IsLOD2() -> (IsLOD2:bool)) -> (Pose) {
    alphaBoolBlend {blendOption=Linear}
    controlRigAssetReference {blueprintRigClass=&/M_MED_WeaveHarbor_CtrlRig.M_MED_WeaveHarbor_CtrlRig_C}
   }
   n10 = Anim Dynamics - Bone: dyn_bottle!(ComponentPose=Local To Component(LocalPose=@n9), Alpha=1, GravityOverride:Vector="0,0,0", ExternalForce="0,0,0") -> (Pose) {
    angularDampingOverride=0.99
    angularSpringConstant=5
    bAngularSpring=true
    bLinearSpring=true
    bOverrideAngularDamping=true
    bOverrideLinearDamping=true
    bUseGravityOverride=true
    linearDampingOverride=0.99
    linearSpringConstant=5
    simulationSpace=BoneRelative
    alphaBoolBlend {blendOption=Linear}
    boundBone {boneName=dyn_bottle}
    componentAppliedLinearAccClamp=<10000 10000 10000>
    componentLinearAccScale=<1.2 1.2 1.2>
    componentLinearVelScale=<1.2 1.2 1.2>
    physicsBodyDefinitions [{
      boundBone {boneName=dyn_bottle}
      boxExtents=<6 6 18>
      constraintSetup {angularLimitsMax=<5 2 3>, angularLimitsMin=<-5 -2 -3>, angularTarget=<1 0 0>, linearAxesMax=<0 0 0.2>, linearAxesMin=<0 0 -0.1>}
      localJointOffset=<0 0 -0.1>
     }]
    relativeSpaceBone {boneName=thigh_twist_01_l}
   }
   n11 = RigidBody!(ComponentPose=@n10, Alpha=Get calfBend_l() -> (calfBend_l:double), OverrideWorldGravity:Vector="0,0,0", ExternalForce) -> (Pose) {
    bOverrideWorldGravity=true
    overridePhysicsAsset=&/M_MED_WeaveHarbor_LeftLeg_Physics
    simulationSpace=BaseBoneSpace
    alphaBoolBlend {blendOption=Linear}
    baseBoneRef {boneName=thigh_l}
    componentAppliedLinearAccClamp=<6000 6000 6000>
    componentLinearAccScale=<0.5 0.5 0.5>
    componentLinearVelScale=<0.5 0.5 0.5>
    simSpaceSettings {maxAngularAcceleration=300, maxAngularVelocity=30, maxLinearAcceleration=100, maxLinearVelocity=10, velocityScaleZ=0.2, worldAlpha=0.7}
   }
   n12 = RigidBody(ComponentPose=@n11, Alpha=Get calfBend_r() -> (calfBend_r:float), OverrideWorldGravity="0,0,0") {
    bOverrideWorldGravity=true
    overridePhysicsAsset=&/M_MED_WeaveHarbor_RightLeg_Physics
    simulationSpace=BaseBoneSpace
    alphaBoolBlend {blendOption=Linear}
    baseBoneRef {boneName=thigh_r}
    componentAppliedLinearAccClamp=<6000 6000 6000>
    componentLinearAccScale=<0.5 0.5 0.5>
    componentLinearVelScale=<0.5 0.5 0.5>
    simSpaceSettings {maxAngularAcceleration=300, maxAngularVelocity=30, maxLinearAcceleration=100, maxLinearVelocity=10, velocityScaleZ=0.2, worldAlpha=0.7}
   }
   n13 = "Constraint - dyn_CircleFX_l_upper (PPPPPPPP)"!(
    ComponentPose=@n12
    Alpha=1
    ConstraintWeights_0=0.0077
    ConstraintWeights_1:float=0.0459
    ConstraintWeights_2:float=0.0773
    ConstraintWeights_3:float=0.7203
    ConstraintWeights_4:float=0.0174
    ConstraintWeights_5:float=0.012
    ConstraintWeights_6:float=0.0059
    ConstraintWeights_7:float=0.1135
   ) -> (Pose) {
    alphaBoolBlend {blendOption=Linear}
    boneToModify {boneName=dyn_CircleFX_l_upper}
    constraintSetup [
     {targetBone {boneName=spine_04}}
     {targetBone {boneName=spine_05}}
     {targetBone {boneName=clavicle_l}}
     {targetBone {boneName=upperarm_twist_01_l}}
     {targetBone {boneName=deform_upperarm_bk_l}}
     {targetBone {boneName=deform_upperarm_fr_base_l}}
     {targetBone {boneName=deform_upperarm_fr_l}}
     {targetBone {boneName=deform_upperarm_up_l}}
    ]
    constraintWeights [1,1,1,1,1,1,1,1]
   }
   n14 = "Constraint - dyn_CircleFX_l_lower (PPPPPPPP)"!(
    ComponentPose=@n13
    Alpha=1
    ConstraintWeights_0=0.1826
    ConstraintWeights_1:float=0.1158
    ConstraintWeights_2:float=0.0863
    ConstraintWeights_3:float=0.3072
    ConstraintWeights_4:float=0.0764
    ConstraintWeights_5:float=0.0354
    ConstraintWeights_6:float=0.0472
    ConstraintWeights_7:float=0.1491
   ) -> (Pose) {
    alphaBoolBlend {blendOption=Linear}
    boneToModify {boneName=dyn_CircleFX_l_lower}
    constraintSetup [
     {targetBone {boneName=upperarm_l}}
     {targetBone {boneName=lowerarm_l}}
     {targetBone {boneName=lowerarm_twist_01_l}}
     {targetBone {boneName=lowerarm_twist_02_l}}
     {targetBone {boneName=deform_elbow_in_l}}
     {targetBone {boneName=deform_elbow_out_l}}
     {targetBone {boneName=upperarm_twist_01_l}}
     {targetBone {boneName=upperarm_twist_02_l}}
    ]
    constraintWeights [1,1,1,1,1,1,1,1]
   }
   n15 = "Constraint - dyn_CircleFX_leg_r (PPPPPPP)"!(
    ComponentPose=@n14
    Alpha=1
    ConstraintWeights_0=0.0021
    ConstraintWeights_1:float=0.8825
    ConstraintWeights_2:float=0.0095
    ConstraintWeights_3:float=0.1051
    ConstraintWeights_4:float=0.0006
    ConstraintWeights_5:float=0.0002
    ConstraintWeights_6:float=0
   ) -> (Pose) {
    alphaBoolBlend {blendOption=Linear}
    boneToModify {boneName=dyn_CircleFX_leg_r}
    constraintSetup [
     {targetBone {boneName=thigh_r}}
     {targetBone {boneName=calf_r}}
     {targetBone {boneName=calf_twist_01_r}}
     {targetBone {boneName=calf_twist_02_r}}
     {targetBone {boneName=deform_knee_in_r}}
     {targetBone {boneName=deform_knee_out_r}}
     {targetBone {boneName=thigh_twist_01_r}}
    ]
    constraintWeights [1,1,1,1,1,1,1]
   }
   n16 = "Constraint - dyn_CircleFX_r_upper (PPPPPPPP)"!(
    ComponentPose=@n15
    Alpha=1
    ConstraintWeights_0=0.0077
    ConstraintWeights_1:float=0.0459
    ConstraintWeights_2:float=0.0773
    ConstraintWeights_3:float=0.7203
    ConstraintWeights_4:float=0.0174
    ConstraintWeights_5:float=0.012
    ConstraintWeights_6:float=0.0059
    ConstraintWeights_7:float=0.1135
   ) -> (Pose) {
    alphaBoolBlend {blendOption=Linear}
    boneToModify {boneName=dyn_CircleFX_r_upper}
    constraintSetup [
     {targetBone {boneName=spine_04}}
     {targetBone {boneName=spine_05}}
     {targetBone {boneName=clavicle_r}}
     {targetBone {boneName=upperarm_twist_01_r}}
     {targetBone {boneName=deform_upperarm_bk_r}}
     {targetBone {boneName=deform_upperarm_fr_base_r}}
     {targetBone {boneName=deform_upperarm_fr_r}}
     {targetBone {boneName=deform_upperarm_up_r}}
    ]
    constraintWeights [1,1,1,1,1,1,1,1]
   }
   n17 = "Constraint - dyn_CircleFX_r_lower (PPPPPPPP)"!(
    ComponentPose=@n16
    Alpha=1
    ConstraintWeights_0=0.1754
    ConstraintWeights_1:float=0.1097
    ConstraintWeights_2:float=0.0942
    ConstraintWeights_3:float=0.2986
    ConstraintWeights_4:float=0.0771
    ConstraintWeights_5:float=0.0375
    ConstraintWeights_6:float=0.0508
    ConstraintWeights_7:float=0.1567
   ) -> (Pose) {
    alphaBoolBlend {blendOption=Linear}
    boneToModify {boneName=dyn_CircleFX_r_lower}
    constraintSetup [
     {targetBone {boneName=upperarm_r}}
     {targetBone {boneName=lowerarm_r}}
     {targetBone {boneName=lowerarm_twist_01_r}}
     {targetBone {boneName=lowerarm_twist_02_r}}
     {targetBone {boneName=deform_elbow_in_r}}
     {targetBone {boneName=deform_elbow_out_r}}
     {targetBone {boneName=upperarm_twist_01_r}}
     {targetBone {boneName=upperarm_twist_02_r}}
    ]
    constraintWeights [1,1,1,1,1,1,1,1]
   }
   n18 = "Constraint - dyn_CircleFX_leg_l (PPPPPPP)"!(
    ComponentPose=@n17
    Alpha=1
    ConstraintWeights_0=0.0021
    ConstraintWeights_1:float=0.8825
    ConstraintWeights_2:float=0.0095
    ConstraintWeights_3:float=0.1051
    ConstraintWeights_4:float=0.0006
    ConstraintWeights_5:float=0.0002
    ConstraintWeights_6:float=0
   ) -> (Pose) {
    alphaBoolBlend {blendOption=Linear}
    boneToModify {boneName=dyn_CircleFX_leg_l}
    constraintSetup [
     {targetBone {boneName=thigh_l}}
     {targetBone {boneName=calf_l}}
     {targetBone {boneName=calf_twist_01_l}}
     {targetBone {boneName=calf_twist_02_l}}
     {targetBone {boneName=deform_knee_in_l}}
     {targetBone {boneName=deform_knee_out_l}}
     {targetBone {boneName=thigh_twist_01_l}}
    ]
    constraintWeights [1,1,1,1,1,1,1]
   }
   n19 = "Constraint - dyn_CircleFX_r_upper (P)"(ComponentPose=@n18, Alpha=1, ConstraintWeights_0=1) {
    alphaBoolBlend {blendOption=Linear}
    boneToModify {boneName=dyn_CircleFX_r_upper}
    constraintSetup [{targetBone {boneName=upperarm_r}}]
    constraintWeights [1]
   }
   n20 = "Constraint - dyn_CircleFX_l_upper (P)"(ComponentPose=@n19, Alpha=1, ConstraintWeights_0=1) {
    alphaBoolBlend {blendOption=Linear}
    boneToModify {boneName=dyn_CircleFX_l_upper}
    constraintSetup [{targetBone {boneName=upperarm_l}}]
    constraintWeights [1]
   }
   n21 = "Constraint - dyn_CircleFX_leg_r (P)"(ComponentPose=@n20, Alpha=1, ConstraintWeights_0=1) {
    alphaBoolBlend {blendOption=Linear}
    boneToModify {boneName=dyn_CircleFX_leg_r}
    constraintSetup [{targetBone {boneName=None}}]
    constraintWeights [1]
   }
   n22 = "Constraint - dyn_CircleFX_r_lower (P)"(ComponentPose=@n21, Alpha=1, ConstraintWeights_0=1) {
    alphaBoolBlend {blendOption=Linear}
    boneToModify {boneName=dyn_CircleFX_r_lower}
    constraintSetup [{targetBone {boneName=lowerarm_r}}]
    constraintWeights [1]
   }
   n23 = "Constraint - dyn_CircleFX_l_lower (P)"(ComponentPose=@n22, Alpha=1, ConstraintWeights_0=1) {
    alphaBoolBlend {blendOption=Linear}
    boneToModify {boneName=dyn_CircleFX_l_lower}
    constraintSetup [{targetBone {boneName=lowerarm_l}}]
    constraintWeights [1]
   }
   n24 = "Constraint - dyn_CircleFX_leg_l (P)"(ComponentPose=@n23, Alpha=1, ConstraintWeights_0=1) {
    alphaBoolBlend {blendOption=Linear}
    boneToModify {boneName=dyn_CircleFX_leg_l}
    constraintSetup [{targetBone {boneName=None}}]
    constraintWeights [1]
   }
   Output Pose(Result=Component To Local(ComponentPose=@n24)) {layerGroup=None}
   Get Is Crouching() -> (Is Crouching:double)
   Get Is Not Knee Sliding() -> (Is Not Knee Sliding:double)
  ]}
 EventGraph {nodes [
   n1 = Event BlueprintUpdateAnimation()
   > Set Delta Time X(Delta Time X=@n1.DeltaTimeX)
   > n2 = Gravity Override State Function DEPRECATED(CharacterPartAnimInstance=Self-Reference(), GameTimeInSeconds=GetGameTimeInSeconds(), GravityOverrideParams=Get Gravity Override Params())
   > Set Gravity Override(Gravity Override=@n2)
   n3 = GetOwnerSkeletalMeshComponent()
   > Set calfBend_r(calfBend_r:float=FInterpTo(
     Current=Get calfBend_r()
     Target=GetAlphaFromReferenceTransformDelta(SkeletalMeshComponent=@n3, BoneName=calf_r, BaseName=thigh_r, InputRange=(X=85,Y=140), OutputRange=(X=1,Y=0), TwistAxis="1,0,0")
     DeltaTime=Get Delta Time X()
     InterpSpeed=10
    )) -> (Output_Get:float)
   n4 = GetOwnerSkeletalMeshComponent()
   > Set calfBend_l(calfBend_l:double=FInterpTo(
     Current=Get calfBend_l()
     Target=GetAlphaFromReferenceTransformDelta(SkeletalMeshComponent=@n4, BoneName=calf_l, BaseName=thigh_l, InputRange=(X=85,Y=140), OutputRange=(X=1,Y=0), TwistAxis="1,0,0")
     DeltaTime=Get Delta Time X()
     InterpSpeed=10
    )) -> (Output_Get:double)
   > n5 = Set IsLOD2(IsLOD2:bool="integer >= integer"(A=Get CurrentLOD(), B=2)) -> (Output_Get:bool)
   > n6 = Branch(Condition=@n5)
   > Set High LOD Alpha(High LOD Alpha:double=0) -> (Output_Get:double)
   > Set Low LOD Alpha(Low LOD Alpha:double=1) -> (Output_Get:double)
   Set High LOD Alpha(execute=@n6.else, High LOD Alpha=1)
   > Set Low LOD Alpha(Low LOD Alpha=0)
   Set Delta Time X(Delta Time X=0)
   n7 = Get Delta Time X()
   > Set AnimDynamicsShortsAlpha(AnimDynamicsShortsAlpha:double=SelectFloat(A=0.6, B=MapRangeClamped(Value=@n7, InRangeA=0.016, InRangeB=0.04, OutRangeA=1, OutRangeB=0), bPickA=Get bIsFalling())) -> (Output_Get:double)
   n8 = Get bIsSkyDiving()
   Set Bottle falling(Bottle falling:bool=AND Boolean(A=Get bIsFalling(), B=NOT Boolean(A=@n8))) -> (Output_Get:bool)
   > Set Not in Sky Diving(Not in Sky Diving:bool=NOT Boolean(A=Get bIsSkyDiving())) -> (Output_Get:bool)
   n9 = Get bIsCrouching()
   > Set Is Crouching(Is Crouching:double=FInterpTo(Current=Get Is Crouching(), Target="To Float (Boolean)"(InBool=@n9), DeltaTime=Get Delta Time X(), InterpSpeed=10)) -> (Output_Get:double)
   "To Float (Boolean)"(InBool=NOT Boolean(A=Get bIsKneeSliding()))
   Set Is Not Knee Sliding(Is Not Knee Sliding:double=0) -> (Output_Get:double)
  ]}
]
variables [
 AnimDynamicsShortsAlpha {type=number}
 Bottle falling {type=boolean}
 calfBend_l {type=number}
 calfBend_r {type=number}
 Delta Time X {type=number}
 Gravity Override {type=Vector}
 Gravity Override Params {type=GravityOverrideParamsStruct}
 High LOD Alpha {type=number}
 Is Crouching {type=number}
 Is Not Knee Sliding {type=number}
 IsLOD2 {type=boolean}
 Low LOD Alpha {type=number}
 Not in Sky Diving {type=boolean}
]
